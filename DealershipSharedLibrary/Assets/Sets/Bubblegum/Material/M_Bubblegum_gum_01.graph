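# Translucent unlit bubblegum material for particle effects, with usage flags for sprites, ribbons, beam trails and mesh particles.
# Emissive adds a flare texture to particle color shaped by two Fresnel terms, modulated by tiled noise; opacity follows particle alpha.
# A switchable world position offset ripples along Z using two scrolling sine waves, masked toward UV edges and scaled by a dynamic parameter.
# With the switch off, that branch yields the raw vertex normal rather than zero.
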Material /DealershipSharedLibrary/Assets/Sets/Bubblegum/Material/M_Bubblegum_gum_01
graph {
 properties {
  bUsedWithBeamTrails=true
  bUsedWithMeshParticles=true
  bUsedWithNanite=true
  bUsedWithNiagaraMeshParticles=true
  bUsedWithNiagaraRibbons=true
  bUsedWithNiagaraSprites=true
  bUsedWithParticleSprites=true
  blendMode=BLEND_Translucent
  pixelDepthOffsetMode=PDOM_Legacy
  shadingModel=MSM_Unlit
  usedShadingModels=Unlit
 }
 nodes [
  n1 = TextureCoordinate()
  n2 = ScalarParameter() {defaultValue=1, group=WPO_Mask, parameterName=PowerMiddle2, sortPriority=0}
  n3 = Multiply(A=@n1, B=OneMinus(None=Power(Base=@n1, Exp=@n2)))
  n4 = ComponentMask(None=@n3) {r=true}
  n5 = ScalarParameter() {defaultValue=7, group=WPO_Mask, parameterName=MultPower2Honzontal, sortPriority=1}
  n6 = ComponentMask(None=@n3) {g=true}
  n7 = ScalarParameter() {defaultValue=6, group=WPO_Mask, parameterName=MultPower2Vertical, sortPriority=2}
  NamedRerouteDeclaration(None=Multiply(A=Multiply(A=@n4, B=@n5), B=Multiply(A=@n6, B=@n7))) {name=WPO_Mask, nodeColor=<rgba 0 0.9734 1 1>}
  n8 = Constant()
  n9 = AppendVector(A=@n8, B=@n8)
  n10 = Divide(
   A=ComponentMask(None=TextureCoordinate()) {r=true}
   B=ScalarParameter() {defaultValue=0.8, group=WPO, parameterName=CountWaves2, sortPriority=1}
  )
  n11 = Time()
  n12 = Add(A=@n10, B=Multiply(A=ScalarParameter() {defaultValue=0.5, group=WPO, parameterName=HorizontalSpeed, sortPriority=2}, B=OneMinus(None=@n11)))
  n13 = Add(
   A=@n10
   B=Multiply(A=ScalarParameter() {defaultValue=-0.2, group=WPO, parameterName=HorizontalSpeed2, sortPriority=3}, B=OneMinus(None=@n11))
  )
  n14 = Abs(None=Multiply(A=Sine(None=@n12), B=Sine(None=@n13)))
  n15 = Constant() {r=1.2}
  n16 = Multiply(A=Multiply(A=DynamicParameter() {paramNames [WPO,R,G,B]}.WPO, B=Power(Base=@n14, Exp=@n15)), B=NamedRerouteUsage())
  NamedRerouteDeclaration(None=StaticSwitchParameter(True=AppendVector(A=@n9, B=@n16), False=VertexNormalWS()) {defaultValue=true, group=WPO, parameterName=Use WPO, sortPriority=0}) {name=WPO, nodeColor=<rgba 1 0.9047 0 1>}
  n17 = TextureSampleParameter2D() {
   group=Emissive
   parameterName=GumLight_T
   sortPriority=1
   texture=/DealershipSharedLibrary/Assets/Sets/Bubblegum/Texture/T_Bubblegum_Flare_RGB
  }
  n18 = ParticleColor()
  n19 = Multiply(
   A=MaterialFunctionCall(Invert Fresnel=StaticBool() {value=true}, Power=Constant() {r=1.7}) {materialFunction=&Fresnel_Function}.Result
   B=Constant() {r=0.5}
  )
  n20 = Multiply(
   A=MaterialFunctionCall(Invert Fresnel=StaticBool(), Power=Constant() {r=3}) {materialFunction=&Fresnel_Function}.Result
   B=Constant() {r=0.7}
  )
  n21 = Add(A=Multiply(A=@n17.B, B=Constant() {r=0.5}), B=Multiply(A=@n18.RGB, B=Add(A=@n19, B=@n20)))
  n22 = TextureSampleParameter2D(UVs=TextureCoordinate() {uTiling=3, vTiling=3}) {
   group=Emissive
   parameterName=CloudsNoise_RGB_T
   samplerType=SAMPLERTYPE_Grayscale
   sortPriority=0
   texture=&LowResBlurredNoise
  }
  n23 = Power(Base=@n22.B, Exp=Constant() {r=0.06})
  NamedRerouteDeclaration(None=MaterialFunctionCall(Emissive=Multiply(A=@n21, B=@n23)) {materialFunction=&MF_FX_EmissiveAdjustments}.Result) {name=Emissive Color, nodeColor=<rgba 1 0 0 1>}
  NamedRerouteDeclaration(None=Saturate(None=@n18.A)) {name=Opacity, nodeColor=<rgba 0 1 0 1>}
  NamedRerouteUsage()
  NamedRerouteUsage()
  NamedRerouteUsage()
 ]
}
