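# Masked, lit bubblegum effect material for Niagara sprites, ribbons and mesh particles, with color tinted by particle color.
# Color blends a desaturated cubemap reflection with a Fresnel-weighted highlight mask, using combined texture and swirl-noise normals.
# A texture channel, particle alpha and a dynamic Dissolve value feed a hard-threshold opacity mask.
# Optional world position offset multiplies two scrolling sine waves along X, scaled by dynamic Strength and a UV-derived mask.
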Material /DealershipSharedLibrary/Assets/Sets/Bubblegum/Material/M_Bubblegum_gum_02
graph {
 properties {
  bAutomaticallySetUsageInEditor=false
  bUsedWithNiagaraMeshParticles=true
  bUsedWithNiagaraRibbons=true
  bUsedWithNiagaraSprites=true
  blendMode=BLEND_Masked
  usedShadingModels=Default Lit
 }
 nodes [
  MaterialFunctionCall(Emissive=NamedRerouteUsage()) {materialFunction=&MF_FX_EmissiveAdjustments}
  n1 = Divide(
   A=ComponentMask(None=TextureCoordinate()) {r=true}
   B=ScalarParameter() {defaultValue=0.8, group=WPO, parameterName=CountWaves, sortPriority=1}
  )
  n2 = Time()
  n3 = Add(A=@n1, B=Multiply(A=ScalarParameter() {defaultValue=0.2, group=WPO, parameterName=HorizontalSpeed, sortPriority=2}, B=OneMinus(None=@n2)))
  n4 = Add(A=@n1, B=Multiply(A=ScalarParameter() {defaultValue=-0.05, group=WPO, parameterName=HorizontalSpeed2, sortPriority=3}, B=OneMinus(None=@n2)))
  n5 = Abs(None=Multiply(A=Sine(None=@n3), B=Sine(None=@n4)))
  n6 = Constant() {r=1.2}
  n7 = Multiply(
   A=Multiply(A=DynamicParameter() {defaultValue=<rgba 0.5 1 1 1>, paramNames [Dissolve,Strength,Power,Opacity]}.Strength, B=Power(Base=@n5, Exp=@n6))
   B=NamedRerouteUsage()
  )
  n8 = Constant()
  NamedRerouteDeclaration(None=StaticSwitchParameter(
    True=MaterialFunctionCall(X=@n7, Y=@n8, Z=@n8) {materialFunction=&MakeFloat3}.Result
    False=VertexNormalWS()
   ) {defaultValue=true, group=WPO, parameterName=Use WPO, sortPriority=0}) {name=WPO, nodeColor=<rgba 1 0.8879 0.0185 1>}
  n9 = TextureCoordinate()
  n10 = ScalarParameter() {defaultValue=1, group=WPO_Mask, parameterName=PowerMiddle2, sortPriority=0}
  n11 = Multiply(A=@n9, B=OneMinus(None=Power(Base=@n9, Exp=@n10)))
  n12 = ComponentMask(None=@n11) {r=true}
  n13 = ScalarParameter() {defaultValue=7, group=WPO_Mask, parameterName=MultPower2Honzontal, sortPriority=1}
  n14 = ComponentMask(None=@n11) {g=true}
  n15 = ScalarParameter() {defaultValue=6, group=WPO_Mask, parameterName=MultPower2Vertical, sortPriority=2}
  NamedRerouteDeclaration(None=Multiply(A=Multiply(A=@n12, B=@n13), B=Multiply(A=@n14, B=@n15))) {name=Mask_WPO, nodeColor=<rgba 0 0.9911 1 1>}
  n16 = TextureSampleParameter2D(UVs=TextureCoordinate()) {
   group=Dissolve
   parameterName=Tex_Dissolve
   sortPriority=0
   texture=/DealershipSharedLibrary/Assets/Sets/Bubblegum/Texture/T_Bubblegum_Mask_RGB
  }
  n17 = ChannelMaskParameter(None=@n16.RGBA) {group=Dissolve, maskChannel=Green, parameterName=Dissolve_Channel, sortPriority=1, defaultValue=<rgba 0 1 0 0>}
  n18 = Multiply(A=@n17.R, B=Constant() {r=1})
  n19 = Constant()
  n20 = Constant() {r=1}
  n21 = Power(Base=ScalarParameter() {defaultValue=2, group=Dissolve, parameterName=Strange, sortPriority=4}, Exp=Constant() {r=2})
  n22 = ComponentMask(None=MaterialFunctionCall(Input=OneMinus(None=@n18), Input Low=@n19, Input High=@n20, Target Low=OneMinus(None=@n21), Target High=@n21) {materialFunction=&RemapValueRange}.Result) {r=true}
  n23 = Multiply(A=ParticleColor().A, B=ScalarParameter() {defaultValue=1, group=Dissolve, parameterName=Water, sortPriority=5})
  n24 = Multiply(A=Constant() {r=12}, B=DynamicParameter() {defaultValue=<rgba 0.5 1 1 1>, paramNames [Dissolve,Strength,Power,Opacity]}.Dissolve)
  n25 = ComponentMask(None=MaterialFunctionCall(Input=OneMinus(None=@n23), Input Low=@n19, Input High=@n20, Target Low=OneMinus(None=@n24), Target High=@n24) {materialFunction=&RemapValueRange}.Result) {r=true}
  n26 = Constant() {r=1}
  n27 = If(A=Constant() {r=0.8}, B=Saturate(None=Add(A=@n22, B=@n25)), "A > B"=@n26, "A == B"=@n26, "A < B"=Constant())
  n28 = DynamicParameter() {defaultValue=<rgba 0.5 1 1 1>, paramNames [Dissolve,Strength,Power,Opacity]}
  NamedRerouteDeclaration(None=Multiply(A=Constant4Vector() {constant=<rgba 1 1 1 1>}.RGB, B=Multiply(A=@n27, B=@n28.Opacity))) {name=Opacity Mask, nodeColor=<rgba 0 1 0 1>}
  n29 = NamedRerouteUsage()
  n30 = Multiply(A=Add(A=NamedRerouteUsage(), B=NamedRerouteUsage()), B=Constant4Vector() {constant=<rgba 1 1 1 1>}.RGB)
  NamedRerouteDeclaration(None=Multiply(A=Add(A=@n29, B=@n30), B=ScalarParameter() {defaultValue=1, group=Normal, parameterName=NormalHeight, sortPriority=2})) {name=Normal, nodeColor=<rgba 0 0 1 1>}
  n31 = TextureSampleParameterCube!(UVs=Transform(None=ReflectionVectorWS()), Apply View MipBias) {
   group=Color
   parameterName=CubeMap
   sortPriority=0
   texture=/Game/Packages/Fortress_Sky/SkyDome/MaterialInstances/SkyDome_Apollo/Cubemap_Apollo_Day
  }
  n32 = Desaturation(None=@n31.RGB)
  n33 = Saturate(None=ConstantBiasScale(None=@n32))
  n34 = Constant() {r=2}
  n35 = Constant3Vector() {constant=<rgba 0 0 1 0>}
  n36 = CameraVectorWS()
  n37 = Saturate(None=Multiply(A=Constant() {r=2}, B=DotProduct(A=@n35, B=@n36)))
  n38 = Add(A=Multiply(A=OneMinus(None=@n37), B=Constant() {r=2}), B=Constant() {r=1})
  n39 = Multiply(A=Multiply(A=@n32, B=ParticleColor().RGB), B=@n38)
  n40 = Constant() {r=0.3}
  n41 = Multiply(A=Add(A=Multiply(A=@n33, B=@n34), B=Multiply(A=@n39, B=@n40)), B=Constant() {r=0.75})
  n42 = Add(A=NamedRerouteUsage(), B=NamedRerouteUsage())
  n43 = Desaturation(None=@n42) {luminanceFactors=<rgba 0.35 0.35 0.3 1>}
  n44 = Power(Base=NamedRerouteUsage(), Exp=DynamicParameter() {defaultValue=<rgba 0.5 1 1 1>, paramNames [Dissolve,Strength,Power,Opacity]}.Power)
  n45 = Saturate(None=OneMinus(None=Fresnel(Normal=@n42) {exponent=3}))
  n46 = Multiply(A=@n38, B=Multiply(A=ParticleColor().RGB, B=LinearInterpolate(A=@n43, B=@n44, Alpha=@n45)))
  NamedRerouteDeclaration(None=LinearInterpolate(A=Multiply(A=@n41, B=@n46), B=@n46, Alpha=@n37)) {name=Color, nodeColor=<rgba 1 0 0 1>}
  n47 = TextureCoordinate()
  n48 = TextureSampleParameter2D(UVs=@n47) {group=Color, parameterName=Texture, sortPriority=1, texture=&T_Bubblegum_Mask_RGB}
  NamedRerouteDeclaration(None=ChannelMaskParameter(None=@n48.RGBA) {group=Color, parameterName=Helight_Channel, sortPriority=2}.R) {name=Texture_Helight, nodeColor=<rgba 0.0782 0 1 1>}
  n49 = TextureSampleParameter2D(UVs=TextureCoordinate() {coordinateIndex=1, uTiling=0.9, vTiling=0.9}) {
   group=Normal
   parameterName=Noise_Normal
   samplerType=SAMPLERTYPE_Normal
   sortPriority=0
   texture=/Game/Global/Textures/Default/Noise/T_SwirlNoise_N
  }
  NamedRerouteDeclaration(None=@n49.RGB) {name=Noise_Normal, nodeColor=<rgba 0.4762 0.0002 1 1>}
  n50 = TextureSampleParameter2D(UVs=@n47) {
   group=Normal
   parameterName=Tex_Normal
   samplerType=SAMPLERTYPE_Normal
   sortPriority=1
   texture=/DealershipSharedLibrary/Assets/Sets/Bubblegum/Texture/T_Bubblegum_Normal_01
  }
  NamedRerouteDeclaration(None=@n50.RGB) {name=Tex_Normal, nodeColor=<rgba 1 0.0144 0.2815 1>}
  NamedRerouteUsage()
  NamedRerouteUsage()
  NamedRerouteUsage()
 ]
}
